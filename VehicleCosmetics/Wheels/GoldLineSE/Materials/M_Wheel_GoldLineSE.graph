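# Masked material for the GoldLineSE cosmetic vehicle wheel, adding animated custom-colour glow over a near-black base from MF_WheelParent_Simple.
# Procedural UV1 rotators and radial pulses, a world-space parallax function and a panned-gradient Fresnel tire effect are declared as named reroutes alongside rim and tire masks.
# Emissive blends rim and tire glow by vertex-colour red, with an optional TertiaryColor switch.
# The attributes then pass through vehicle damage and de-rez transition functions.
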Material /VehicleCosmetics/Wheels/GoldLineSE/Materials/M_Wheel_GoldLineSE
graph {
 properties {
  bAutomaticallySetUsageInEditor=false
  bUseMaterialAttributes=true
  bUsedWithMeshDeformer=true
  bUsedWithSkeletalMesh=true
  blendMode=BLEND_Masked
  pixelDepthOffsetMode=PDOM_Legacy
  usedShadingModels=Default Lit
 }
 nodes [
  n1 = MaterialFunctionCall() {materialFunction=&MF_WheelParent_Simple}
  GetMaterialAttributes(None=@n1.MaterialAttributes)
  n2 = VectorParameter() {group=5_Special Effects, parameterName=CustomColor2, sortPriority=33, defaultValue=<rgba 1 0.65 0 1>}
  n3 = Multiply(A=NamedRerouteUsage(), B=NamedRerouteUsage())
  n4 = VectorParameter() {group=5_Special Effects, parameterName=CustomColor1, sortPriority=34, defaultValue=<rgba 0.027 0.027 0.027 1>}
  n5 = LinearInterpolate(
   A=@n4.RGB
   B=Multiply(A=@n2.RGB, B=ScalarParameter() {defaultValue=20, group=5_Special Effects, parameterName=CarbonGlowStrength, sortPriority=36}) {constA=2, constB=10}
   Alpha=NamedRerouteUsage()
  )
  n6 = LinearInterpolate(A=@n5, B=VectorParameter() {parameterName=TertiaryColor, defaultValue=<rgba 1 0 0.74 0>}.RGB) {constAlpha=0}
  n7 = OneMinus(None=NamedRerouteUsage())
  n8 = NamedRerouteUsage()
  n9 = NamedRerouteUsage()
  n10 = ScalarParameter() {defaultValue=1, group=5_Special Effects, parameterName=GlowPulsingPower, sortPriority=38}
  n11 = Frac(None=OneMinus(None=NamedRerouteUsage()))
  n12 = Multiply(A=@n11, B=@n11)
  n13 = Add(A=Add(A=@n3, B=Multiply(A=@n7, B=@n8)), B=Multiply(A=Multiply(A=@n9, B=@n10), B=OneMinus(None=@n12)))
  n14 = Add(
   A=Multiply(
    A=Multiply(A=@n3, B=@n2.RGB)
    B=ScalarParameter() {defaultValue=20, group=5_Special Effects, parameterName=CarbonGlowStrength, sortPriority=36}
   )
   B=LinearInterpolate(
    A=StaticSwitchParameter(True=@n5, False=@n6) {group=5_Special Effects, parameterName=UseTertiaryColor, sortPriority=42}
    B=VectorParameter() {group=5_Special Effects, parameterName=CustomColor3, sortPriority=35, defaultValue=<rgba 0 0 0 1>}.RGB
    Alpha=LinearInterpolate(Alpha=@n13) {constA=1, constB=0}
   )
  )
  n15 = Add(A=NamedRerouteUsage(), B=ScalarParameter() {group=5_Special Effects, parameterName=WheelReflectionBrightness, sortPriority=43})
  n16 = Multiply(A=@n4.RGB, B=Add(A=NamedRerouteUsage(), B=NamedRerouteUsage()))
  n17 = ScalarParameter() {defaultValue=10, group=5_Special Effects, parameterName=TireGlowStrength, sortPriority=44}
  n18 = Multiply(
   A=LinearInterpolate(A=Multiply(A=@n14, B=@n15), B=Multiply(A=@n16, B=@n17), Alpha=VertexColor().R)
   B=ScalarParameter() {defaultValue=1, group=5_Special Effects, parameterName=GlobalBrightness, sortPriority=37}
  )
  n19 = Constant3Vector() {constant=<rgba 0.02 0.02 0.02 0>}
  MaterialFunctionCall(In=MaterialFunctionCall(Material=SetMaterialAttributes!(MaterialAttributes=@n1.MaterialAttributes, Emissive Color=@n18, Base Color=@n19) {attributeSetTypes [#1, #2]}) {materialFunction=&MF_BaseVehicle_Damage}.Result) {materialFunction=&MF_VehicleDeRez_Transition}
  n20 = Multiply(A=Constant3Vector() {constant=<rgba 1 0 0 0>}) {constB=2}
  n21 = Subtract(A=TextureCoordinate() {coordinateIndex=1}) {constB=0.5}
  n22 = ComponentMask(None=@n21) {r=true}
  n23 = ComponentMask(None=@n21) {g=true}
  n24 = Time()
  n25 = ScalarParameter() {defaultValue=0.2, group=5_Special Effects, parameterName=RotationSpeed, sortPriority=41}
  n26 = Multiply(
   A=Divide(
    A=ScalarParameter() {defaultValue=5, group=5_Special Effects, parameterName=NumberOfRotators, sortPriority=39}
    B=Multiply(A=@n20) {constB=3.142}
   )
   B=Add(A=Arctangent2!(Y=@n22, X=@n23), B=Multiply(A=@n24, B=@n25))
  )
  n27 = Add(A=ComponentMask(None=Sine(None=@n26)) {r=true}) {constB=0.4}
  n28 = SmoothStep(Value=Clamp(None=Divide(A=@n27)) {maxDefault=2}) {constMax=10}
  NamedRerouteDeclaration(None=Multiply(
    A=Multiply(A=@n28) {constA=6, constB=5}
    B=ScalarParameter() {defaultValue=20, group=5_Special Effects, parameterName=CarbonGlowStrength, sortPriority=36}
   )) {name=CarbonFX, nodeColor=<rgba 1 0 0.75 1>}
  NamedRerouteDeclaration(None=ComponentMask(None=@n1.TireRGBAMask) {r=true}) {name=TireRMask, nodeColor=<rgba 0 0.3672 1 1>}
  n29 = TextureCoordinate() {coordinateIndex=1}
  n30 = Multiply(A=TextureSample(UVs=Panner(Coordinate=@n29) {speedX=0.5}) {texture=/VehicleCosmetics/Wheels/_SharedTextures/FX/Gradient_D}.R) {constB=2}
  n31 = Fresnel() {exponent=4}
  NamedRerouteDeclaration(None=Clamp(None=Multiply(A=@n30, B=@n31)) {maxDefault=6}) {name=TireFX, nodeColor=<rgba 1 0.2812 0 1>}
  n32 = MaterialFunctionCall(Color=@n1.RimRGBAMask) {materialFunction=&SplitComponents}
  NamedRerouteDeclaration(None=@n32.R) {name=RimRMask, nodeColor=<rgba 0 0.1328 1 1>}
  n33 = Subtract(A=TextureCoordinate() {coordinateIndex=1}) {constB=0.5}
  n34 = Length(None=ComponentMask(None=Multiply(A=@n33) {constB=3}) {g=true, r=true}) {is=class MaterialExpressionLength}
  n35 = Multiply(A=Time(), B=ScalarParameter() {defaultValue=-0.1, group=5_Special Effects, parameterName=PulseSpeed, sortPriority=40})
  n36 = Power(Base=Cosine(None=Add(A=@n34, B=@n35)) {period=0.9}) {constExponent=2.5}
  NamedRerouteDeclaration(None=Clamp(None=LinearInterpolate(Alpha=@n36) {constB=0.1})) {name=PulseFX, nodeColor=<rgba 0 1 0.5078 1>}
  n37 = Constant()
  NamedRerouteDeclaration(None=MaterialFunctionCall!(Desaturate=Constant() {r=1}, Tint=@n37, TintAmount=@n37, Input ParallaxSize) {materialFunction=&/MF_Wheel_GoldLineSE}.WorldSpace) {name=ParallaxWSFX, nodeColor=<rgba 0 1 0.25 1>}
  n38 = Sine(None=Multiply(A=Time()) {constB=0.2})
  NamedRerouteDeclaration(None=Divide(A=Add(A=@n38) {constB=2}) {constB=3}) {name=GlowPulsingFX, nodeColor=<rgba 1 0.3281 0 1>}
  NamedRerouteDeclaration(None=@n32.G) {name=RimGMask, nodeColor=<rgba 0.4531 1 0 1>}
  NamedRerouteDeclaration(None=@n32.B) {name=RimBMask, nodeColor=<rgba 1 0 0.7734 1>}
  NamedRerouteDeclaration(None=ComponentMask(None=@n1.RimRGBAMask) {a=true}) {name=RimAMask, nodeColor=<rgba 1 0.2578 0 1>}
  VectorParameter() {group=5_Special Effects, parameterName=CustomColor, sortPriority=34, defaultValue=<rgba 0.027 0.027 0.027 1>}
  VectorParameter() {group=5_Special Effects, parameterName=AccentColor, sortPriority=33, defaultValue=<rgba 1 0.65 0 1>}
  VectorParameter() {group=5_Special Effects, parameterName=BackgroundColor, sortPriority=35, defaultValue=<rgba 0 0 0 1>}
 ]
}
